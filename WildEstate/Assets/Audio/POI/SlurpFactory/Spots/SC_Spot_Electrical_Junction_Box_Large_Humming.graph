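# Looping ambient spot cue for the hum of a large electrical junction box in the Slurp Factory POI.
# A Random node picks one of three equally weighted looping hum waves, which feeds a default Modulator and then an indefinitely looping Enveloper with a gentle periodic volume dip.
# It uses the indoor ambient sound class, borrows a small Domestic Commercial spot attenuation asset, overrides concurrency with a low voice cap, and has its debug flag on.
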SoundCue /WildEstate/Assets/Audio/POI/SlurpFactory/Spots/SC_Spot_Electrical_Junction_Box_Large_Humming
root=Enveloper_0
nodes [
 {
  id=Enveloper_0
  children [Modulator_0]
  properties {
   bLoop=true
   bLoopIndefinitely=true
   loopEnd=20
   pitchCurve {editorCurveData {keys [{time=0}]}}
   volumeCurve {editorCurveData {keys [<at -0.0028 1>,<at 0 1>,<at 10 0.95>,<at 20 1>]}}
  }
 }
 {id=Modulator_0, children [Random_0]}
 {id=Random_0, children [WavePlayer_2,WavePlayer_3,WavePlayer_5], properties {weights [1,1,1]}}
 {
  id=WavePlayer_2
  properties {bLooping=true, soundWaveAssetPtr=&/Spot_Electrical_Junction_Box_Large_Humming_01}
 }
 {
  id=WavePlayer_3
  properties {bLooping=true, soundWaveAssetPtr=&/Spot_Electrical_Junction_Box_Large_Humming_02}
 }
 {
  id=WavePlayer_5
  properties {bLooping=true, soundWaveAssetPtr=&/Spot_Electrical_Junction_Box_Large_Humming_03}
 }
]
properties {
 attenuationSettings=/WildEstate/Assets/Audio/POI/Domestic_Commercial/Spots/Spot_WildEstate_Domestic_Commercial_Small_Att
 bDebug=true
 bOverrideConcurrency=true
 maxDistance=1100
 soundClassObject=&Fort_Ambient_Indoors_Athena
 volumeMultiplier=1
 attenuationOverrides {bSpatialize=false, stereoSpread=1147.9967, attenuationShapeExtents {x=6826.1992}}
 concurrencyOverrides {maxCount=2}
 concurrencySet [/Game/Concurrency/AmbientConc_3Max_Spot_Default_Concurrency]
}
