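# Looping ambient spot sound for a running vat pump at the Slurp Factory POI.
# A Random node picks one of two equally weighted looping waves, which passes through a Modulator for variation.
# An Enveloper repeats indefinitely on top, adding a slow, slight volume dip and recovery.
# It uses the indoor ambient sound class, the shared vats attenuation asset and a concurrency override capping simultaneous instances at two.
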SoundCue /WildEstate/Assets/Audio/POI/SlurpFactory/Spots/SC_Spot_SlurpFactory_Vat_Pump_Running_Loop
root=Enveloper_0
nodes [
 {
  id=Enveloper_0
  children [Modulator_0]
  properties {
   bLoop=true
   bLoopIndefinitely=true
   loopEnd=20
   pitchCurve {editorCurveData {keys [{time=0}]}}
   volumeCurve {editorCurveData {keys [<at -0.0028 1>,<at 0 1>,<at 10 0.95>,<at 20 1>]}}
  }
 }
 {id=Modulator_0, children [Random_1]}
 {id=Random_1, children [WavePlayer_1,WavePlayer_4], properties {weights [1,1]}}
 {
  id=WavePlayer_1
  properties {bLooping=true, soundWaveAssetPtr=&/Spot_SlurpFactory_Vat_Pump_Running_Loop_02}
 }
 {
  id=WavePlayer_4
  properties {bLooping=true, soundWaveAssetPtr=&/Spot_SlurpFactory_Vat_Pump_Running_Loop_01}
 }
]
properties {
 attenuationSettings=&/Spot_WildEstate_SlurpFactory_Vats_Att
 bDebug=true
 bOverrideConcurrency=true
 maxDistance=2500
 soundClassObject=&Fort_Ambient_Indoors_Athena
 volumeMultiplier=1
 attenuationOverrides {bSpatialize=false, stereoSpread=1147.9967, attenuationShapeExtents {x=6826.1992}}
 concurrencyOverrides {maxCount=2}
 concurrencySet [/AmbientAudioBase/Concurrency/AmbientConc_2Max_Loop_PreventNew_Default_Concurrency]
}
